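# Additive, unlit, two-sided material for laser ring effects on Niagara mesh particles and spline meshes.
# SplineThicken sets width from particle alpha, plus panning Perlin-noise world position wobble that fades along V.
# A swapped-UV gradient, dimmed where the tangent points along the view, drives particle-colored emissive and an alpha blended between Start and End values.
# The curve atlas and dynamic parameter nodes feed no other node.
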Material /WildEstateWeapons/Gameplay/CoveSpice/PH/FX/MI_Laser_Rings
graph {
 properties {
  bCastRayTracedShadows=false
  bUsedWithNiagaraMeshParticles=true
  bUsedWithSplineMeshes=true
  blendMode=BLEND_Additive
  pixelDepthOffsetMode=PDOM_Legacy
  shadingModel=MSM_Unlit
  twoSided=true
  usedShadingModels=Unlit
 }
 nodes [
  n1 = ScalarParameter() {defaultValue=0.5, parameterName=SplineWidthTip}
  n2 = ScalarParameter() {defaultValue=3, parameterName=SplineWidthBase}
  n3 = ParticleColor()
  n4 = Power(Base=@n3.A) {constExponent=0.5}
  n5 = TextureCoordinate()
  n6 = Power(Base=ComponentMask(None=@n5) {g=true}, Exp=ScalarParameter() {defaultValue=2, parameterName=Power})
  n7 = ScalarParameter() {defaultValue=2.5, parameterName=WPO Wobble}
  n8 = ParticleRandom()
  n9 = ScalarParameter() {defaultValue=1, parameterName=WPO Wobble Pan Speed}
  n10 = Constant2Vector() {g=1, r=0.05}
  n11 = Multiply(
   A=Panner(Coordinate=Add(A=@n5, B=@n8), Speed=Multiply(A=@n9, B=@n10)) {speedX=0.1, speedY=2}
   B=ScalarParameter() {defaultValue=2, parameterName=WPO Wobble Tile}
  )
  n12 = Transform(None=ConstantBiasScale(None=TextureSample(UVs=@n11) {samplerType=SAMPLERTYPE_Masks, texture=/Game/Global/Textures/Default/Noise/T_Generic_Perlin}.RGB) {bias=-0.5, scale=2})
  Add(
   A=MaterialFunctionCall(WidthBase=LinearInterpolate(A=@n1, B=@n2, Alpha=@n4), WidthTip=@n1, UVs for Thickness=@n5) {materialFunction=&SplineThicken}.WorldPositionOffset
   B=Multiply(A=OneMinus(None=@n6), B=Multiply(A=@n7, B=@n12))
  )
  n13 = TextureCoordinate()
  CurveAtlasRowParameter(CurveTime=MaterialFunctionCall(Float2=@n13) {materialFunction=&BreakOutFloat2Components}.R) {atlas=/Game/Athena/LinearColorCurves/CA_Contrail_Abduction, curve=/Game/Athena/LinearColorCurves/Curve_Contrail_Abduction_TractorBeam_Ring}
  DynamicParameter()
  n14 = TextureSample(UVs=MaterialFunctionCall(XYZ=@n13) {materialFunction=&Swizzle}.YX) {texture=/Game/Global/Textures/Default/Masks_and_Gradients/gradient_stasis}
  n15 = Abs(None=DotProduct(A=CameraVectorWS(), B=VertexTangentWS()))
  n16 = Saturate(None=Multiply(A=OneMinus(None=@n15)) {constB=2})
  n17 = Multiply(A=Multiply(A=@n14.G, B=@n14.G), B=@n16)
  MaterialFunctionCall(Emissive=Multiply(A=LinearInterpolate(Alpha=@n17) {constA=0.2}, B=@n3.RGB)) {materialFunction=&MF_FX_EmissiveAdjustments}
  n18 = ScalarParameter() {parameterName=End Alpha}
  n19 = ScalarParameter() {defaultValue=4, parameterName=Start Alpha}
  Multiply(A=Multiply(A=Multiply(A=@n14.G, B=@n16), B=LinearInterpolate(A=@n18, B=@n19, Alpha=@n6) {constB=5}), B=@n3.A)
 ]
}
