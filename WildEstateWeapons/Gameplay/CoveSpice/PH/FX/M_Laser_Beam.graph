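# Translucent unlit material for a laser beam effect, usable on sprites, Niagara ribbons and mesh particles.
# A tiled, panning trail mask feeds offset and contrast stages that blend outside, mid and center colors, tinted by particle color through an emissive adjustment function.
# A parallel three-level blend with its own offsets and contrasts shapes opacity, multiplied by particle alpha, depth fade and near-camera fade.
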Material /WildEstateWeapons/Gameplay/CoveSpice/PH/FX/M_Laser_Beam
graph {
 properties {
  bUsedWithMeshDeformer=true
  bUsedWithMeshParticles=true
  bUsedWithNiagaraMeshParticles=true
  bUsedWithNiagaraRibbons=true
  bUsedWithParticleSprites=true
  bUsedWithSkeletalMesh=true
  blendMode=BLEND_Translucent
  pixelDepthOffsetMode=PDOM_Legacy
  shadingModel=MSM_Unlit
  usedShadingModels=Unlit
 }
 nodes [
  n1 = ParticleColor()
  n2 = VectorParameter() {parameterName=Color Outside, defaultValue=<rgba 1 0 0.2577 0>}
  n3 = ScalarParameter() {defaultValue=1, group=Main Gradient, parameterName=GradientTiling}
  n4 = Multiply(A=TextureCoordinate(), B=AppendVector(A=@n3, B=@n3))
  n5 = MaterialFunctionCall(X=ScalarParameter() {defaultValue=3, group=Main Gradient, parameterName=Panner Speed}, Y=Constant()) {materialFunction=&MakeFloat2}
  n6 = NamedRerouteDeclaration(None=TextureSample(UVs=Panner(Coordinate=@n4, Speed=@n5.Result) {speedX=1}) {texture=/Game/Effects/Fort_Effects/Textures/Smoke/T_Lobber_Trail_Mask}.R) {name=Tiling, nodeColor=<rgba 1 0.7266 0 1>}
  n7 = ScalarParameter() {group=Main Gradient, parameterName=Color Offset 1}
  n8 = LinearInterpolate(
   A=VectorParameter() {parameterName=Color Mid, defaultValue=<rgba 0 0.6168 1 0>}.RGB
   B=VectorParameter() {parameterName=Color Center, defaultValue=<rgba 1 1 1 0>}.RGB
   Alpha=MaterialFunctionCall(In=Add(A=@n6, B=@n7) {constB=0.2}, Contrast=ScalarParameter() {defaultValue=0.4, group=Main Gradient, parameterName=Color Contrast 1}) {materialFunction=&CheapContrast}.Result
  )
  n9 = MaterialFunctionCall(
   In=Add(A=NamedRerouteUsage(), B=ScalarParameter() {defaultValue=0.4, group=Main Gradient, parameterName=Color Offset 2}) {constB=0.2}
   Contrast=ScalarParameter() {defaultValue=3, group=Main Gradient, parameterName=Color Contrast 2}
  ) {materialFunction=&CheapContrast}
  MaterialFunctionCall(Emissive=Multiply(A=@n1.RGB, B=LinearInterpolate(A=@n2.RGB, B=@n8, Alpha=@n9.Result))) {materialFunction=&MF_FX_EmissiveAdjustments}
  n10 = MaterialFunctionCall(Distance=ScalarParameter() {defaultValue=100, group=Alpha, parameterName=NearCameraFade}) {materialFunction=&Near_Camera_Fade}
  n11 = ScalarParameter() {defaultValue=35, group=Alpha, parameterName=FadeDistance}
  n12 = ScalarParameter() {parameterName=Alpha Outside}
  n13 = NamedRerouteUsage()
  n14 = ScalarParameter() {defaultValue=0.2, group=Alpha, parameterName=Alpha Offset 1}
  n15 = LinearInterpolate(
   A=ScalarParameter() {defaultValue=1, group=Alpha, parameterName=Alpha Mid}
   B=ScalarParameter() {defaultValue=1, group=Alpha, parameterName=Alpha Center}
   Alpha=MaterialFunctionCall(In=Add(A=@n13, B=@n14) {constB=0.2}, Contrast=ScalarParameter() {defaultValue=2, group=Alpha, parameterName=Alpha Contrast 1}) {materialFunction=&CheapContrast}.Result
  )
  n16 = MaterialFunctionCall(
   In=Add(A=NamedRerouteUsage(), B=ScalarParameter() {group=Alpha, parameterName=Alpha Offset 2}) {constB=0.2}
   Contrast=ScalarParameter() {group=Alpha, parameterName=Alpha Contrast 2}
  ) {materialFunction=&CheapContrast}
  Multiply(
   A=@n1.A
   B=Multiply(A=DepthFade(Opacity=@n10.Result, FadeDistance=@n11) {fadeDistanceDefault=50}, B=LinearInterpolate(A=@n12, B=@n15, Alpha=@n16.Result))
  )
 ]
}
