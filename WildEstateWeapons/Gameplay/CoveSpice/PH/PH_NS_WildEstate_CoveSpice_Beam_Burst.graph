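# Niagara beam burst effect for the CoveSpice weapon, drawn to User.BeamEnd_0, with User.ChargePercent blending beam, rings and flashes from green to orange.
# MainRibbon places partitioned particles along a Hermite spline, adding waveform wobble that is rotated by beam orientation and tapers at both ends.
# RingsMeshes sends looping mesh and glow bursts along that wobbling path, more frequently as beam length grows.
# Flashes, flare and flipbook emitters follow the beam end as impact effects.
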NiagaraSystem /WildEstateWeapons/Gameplay/CoveSpice/PH/PH_NS_WildEstate_CoveSpice_Beam_Burst
dependencies {
 usedDataInterfaces [
  &NiagaraDataInterfaceCurve
  &NiagaraDataInterfaceSpriteRendererInfo
  &NiagaraDataInterfaceColorCurve
 ]
 usedDynamicInputs [&Subtract_Vector, &FindQuatBetweenTwoVectors]
 usedModules [
  &SystemState
  &self:SystemSpawnScript.SystemScriptSource.SystemScriptGraph.NiagaraNodeAssignment_2.SetVariables_#1
  &self:SystemSpawnScript.SystemScriptSource.SystemScriptGraph.NiagaraNodeAssignment_0.SetVariables_#2
  &self:SystemSpawnScript.SystemScriptSource.SystemScriptGraph.NiagaraNodeAssignment_1.SetVariables_#3
 ]
}
emitters {
 Flashes {
  emitterUpdateScript [
   &EmitterState()
   &SpawnRate(SpawnRate=14, Use Spawn Probability=false)
  ]
  particleSpawnScript [
   &InitializeParticle(
    Color=<rgba 10 10 10 1>
    Lifetime Max=0.25
    Lifetime Min=0.1
    Lifetime Mode=::NewEnumerator1
    Mesh Scale=/Niagara/DynamicInputs/CascadeConversion/SplitVector(
     Double Component Vector Channels:ECascadeNiagaraTwoVectorChannels=::NewEnumerator1
     Double Component:NiagaraFloat=&RandomRangeFloat(Minimum=0.4, Maximum=0.5, Evaluation Type=::NewEnumerator1)
     Single Component:NiagaraFloat=&RandomRangeFloat(Minimum=0.175, Maximum=0.225, Evaluation Type=::NewEnumerator1)
    )
    Mesh Scale Mode=::NewEnumerator4
    Position Offset=<0 2 0>
    UsePositionOffset=false
    Write Material Random=true
   )
   &DynamicMaterialParameters(
    Index 0 Param 3=&RandomRangeFloat(Minimum=1, Maximum=1.5, Evaluation Type=::NewEnumerator1)
    Module.Index0Float0Write=true
    Module.Index0Float1Write=true
    Module.Index0Float2Write=true
    Module.Index0Float3Write=true
   )
   &InitialMeshOrientation(Use Rotation Vector=true)
  ]
  particleUpdateScript [
   &ParticleState()
   &CameraOffset(Camera Offset Amount=25)
   &ScaleMeshSize(Scale Factor=&VectorFromFloat(Value=&FloatFromCurve(
      FloatCurve=DataInterface() {propertyValues {Curve {keys [
          {arriveTangent=2.4702, interpMode=RCIM_Cubic, leaveTangent=2.4702, tangentMode=RCTM_User, time=0, value=0}
          {arriveTangent=0.9995, interpMode=RCIM_Cubic, leaveTangent=0.9995, tangentMode=RCTM_User, time=0.4, value=0.75}
          {arriveTangent=-0.0405, interpMode=RCIM_Cubic, leaveTangent=-0.0405, tangentMode=RCTM_User, time=1, value=1}
         ]}}}
      Scale Curve=1.2
     )))
   &ScaleColor(
    Scale Alpha=&FloatFromCurve(FloatCurve=DataInterface() {propertyValues {Curve {keys [
         {arriveTangent=-0.2241, interpMode=RCIM_Cubic, leaveTangent=-0.2241, tangentMode=RCTM_User, time=0.33, value=1}
         {arriveTangent=-3.1123, interpMode=RCIM_Cubic, leaveTangent=-3.1123, tangentMode=RCTM_User, time=1, value=0}
        ]}}})
    Scale RGB=&MakeVectorFromLinearColorRGB(Color=&LerpLinearColors(StartColor=<rgba 0.4301 2 0 1>, EndColor=<rgba 2 0.5689 0 1>, LerpFactor=$User.ChargePercent))
    ScaleA=true
    ScaleRGB=true
   )
   /SetVariables(Particles.Renderer2_Color:LinearColor=&MultiplyLinearColor(Linear Color=$Particles.Color, Scale Factor=<0.05 0.05 0.05 1>)) {isSetParameters=true}
   /SetVariables(Particles.Renderer2_Size:Vector3f=&Multiply_Vector(A=$Particles.Scale, B=<2.1 2.1 2.1>)) {isSetParameters=true}
   /SetVariables(Particles.Renderer2_CamOffset=-5) {isSetParameters=true}
   /SetVariables(Particles.Speedlines2Color:LinearColor=&MakeLinearColorFromVectorAndFloat(
     "Vector (RGB)"=&MakeVectorFromLinearColorRGB(Color=&LerpLinearColors(StartColor=<rgba 0.4301 2 0 1>, EndColor=<rgba 2 0.5689 0 1>, LerpFactor=$User.ChargePercent))
     "Float (Alpha)"=&FloatFromCurve(
      FloatCurve=DataInterface() {propertyValues {Curve {keys [
          {arriveTangent=44.6077, interpMode=RCIM_Cubic, leaveTangent=44.6077, tangentMode=RCTM_User, time=0, value=0}
          <at 0.0781 1.0138 RCIM_Cubic>
          {arriveTangent=-1.8607, interpMode=RCIM_Cubic, leaveTangent=-1.8607, tangentMode=RCTM_User, time=1, value=0}
         ]}}}
      Scale Curve=0.5
     )
    )) {isSetParameters=true}
   /SetVariables(Particles.Speedlines2Dyn:Vector4f=&MakeVector4(X=1, Y=&RandomRangeFloat(Minimum=1.5, Maximum=2), Z=1, W=1)) {isSetParameters=true}
   /SetVariables(Particles.Speedlines2Scale:Vector3f=&Multiply_Vector(
     A=&VectorFromFloat(Value=&Multiply_Float(A=1.75, B=&RandomRangeFloat(Minimum=1, Maximum=1.8)))
     B=&VectorFromFloat(Value=&FloatFromCurve(FloatCurve=DataInterface() {propertyValues {Curve {keys [{arriveTangent=2.9993, interpMode=RCIM_Cubic, leaveTangent=2.9993, tangentMode=RCTM_User, time=0, value=0}, <at 1 1 RCIM_Cubic>]}}}))
    )) {isSetParameters=true}
   /SetVariables(Particles.Speedlines2Quat:Quat4f=&MakeQuaternion(
     Quaternion From=0
     Axis=<1 0 0>
     Angle=&RandomRangeFloat(Minimum=0, Maximum=1)
     From=?
     To=?
     Angle Type=::NewEnumerator1
     Normalize Vectors=false
    )) {isSetParameters=true}
   /SetVariables(Particles.Position=&AddVectorToPosition(
     Position=/Niagara/DynamicInputs/Transforms/TransformPosition(Position=Linked() {linkedVariable {name=User.BeamEnd_0, type=NiagaraPosition}}, Source Space=::World, Destination Space=::Local)
     Vector=<0 2 0>
    )) {isSetParameters=true, module=SetVariables_#4}
  ]
  rendererClasses [&NiagaraMeshRendererProperties]
  renderers [
   {rendererClass=&NiagaraMeshRendererProperties, rendererIndex=0}
   {rendererClass=&NiagaraMeshRendererProperties, rendererIndex=1}
   {rendererClass=&NiagaraMeshRendererProperties, rendererIndex=2}
  ]
 }
 MainRibbon {
  emitterSpawnScript [/Game/Effects/Niagara/Modules/Partition/FN_PartitionAttributes(Number of Bone Sets=1, Particles Per Partition=30)]
  emitterUpdateScript [
   &EmitterState()
   &SpawnBurst_Instantaneous(Spawn Count=$Emitter.FN_PartitionAttributes.Spawn Total, Use Spawn Probability=false)
  ]
  particleSpawnScript [
   &InitializeRibbon(
    Color=&LerpLinearColors(StartColor=<rgba 0.8602 4 0 1>, EndColor=<rgba 6 1.3272 0 1>, LerpFactor=$User.ChargePercent)
    Lifetime=0.7
    Ribbon Width=12
    UsePositionOffset=false
    Write Material Random=true
   )
   /Game/Effects/Niagara/Modules/Partition/FN_Particle_partition_v5/FN_Particle_Partition_v5(Enable Ribbon ID=true)
   /SetVariables(Particles.Beam2_Width=20) {isSetParameters=true}
   /SetVariables(Particles.Renderer2_Color:LinearColor=&LerpLinearColors(StartColor=<rgba 0 0.5 0.0356 1>, EndColor=<rgba 0.6562 0.1649 0 1>, LerpFactor=$User.ChargePercent)) {isSetParameters=true}
   /SetVariables(Particles.RibbonLinkOrder:NiagaraFloat=&ReturnNormalizedExecIndex()) {isSetParameters=true}
   /SetVariables() {isSetParameters=true}
   /SetVariables(Particles.WobbleOverLength:NiagaraFloat=&FloatFromCurve(
     FloatCurve=DataInterface() {propertyValues {Curve {keys [
         {arriveTangent=5.2404, interpMode=RCIM_Cubic, leaveTangent=5.2404, tangentMode=RCTM_User, time=0, value=0}
         <at 0.5 1 RCIM_Cubic>
         {arriveTangent=-4.4363, interpMode=RCIM_Cubic, leaveTangent=-4.4363, tangentMode=RCTM_User, time=1, value=0}
        ]}}}
     CurveIndex=$Particles.Partition.NormalizedID
    )) {isSetParameters=true, module=SetVariables_#5}
  ]
  particleUpdateScript [
   &ParticleState()
   /Game/Effects/Niagara/Modules/Curves/FN_HermiteCubicSpline_v1/FN_HermiteCubicSpline(
    Position Blend:NiagaraFloat=$Particles.Initial.Partition.NormalizedID
    Position End:Vector3f=&TransformPosition(Position=Linked() {linkedVariable {name=User.BeamEnd_0, type=NiagaraPosition}}, Source Space=::World, Destination Space=::Local)
    Position Start:Vector3f=$System.BeamOriginOffset
    Write Particles Position=true
   )
   /SetVariables(Particles.ParticleDistance:NiagaraFloat=&Multiply_Float(
     A=&Abs_Float(Float:NiagaraFloat=&VectorLength(Vector=Linked() {linkedVariable {name=Particles.Position, type=NiagaraPosition}}))
     B=0.1
    )) {isSetParameters=true, module=SetVariables_#6}
   &ScaleRibbonWidth(Ribbon Width Scale=&FloatFromCurve(
     FloatCurve=DataInterface() {propertyValues {Curve {keys [
         {interpMode=RCIM_Cubic, tangentMode=RCTM_User, time=0, value=0.4}
         {arriveTangent=0.75, interpMode=RCIM_Cubic, leaveTangent=0.75, time=0.1, value=1}
         {interpMode=RCIM_Cubic, tangentMode=RCTM_User, time=0.8, value=1}
         {tangentMode=RCTM_User, time=1, value=0}
        ]}}}
     CurveIndex=&ReturnNormalizedExecIndex()
    ))
   /SetVariables(Particles.Soft Glow Ribbon Width:NiagaraFloat=&FloatFromCurve(
     FloatCurve=DataInterface() {propertyValues {Curve {keys [
         {interpMode=RCIM_Cubic, tangentMode=RCTM_User, time=0, value=0.4}
         {arriveTangent=0.75, interpMode=RCIM_Cubic, leaveTangent=0.75, time=0.1, value=1}
         {interpMode=RCIM_Cubic, tangentMode=RCTM_User, time=0.8, value=1}
         {tangentMode=RCTM_User, time=1, value=0}
        ]}}}
     CurveIndex=&ReturnNormalizedExecIndex()
     Scale Curve=50
    )) {isSetParameters=true}
   /SetVariables(Particles.Position=&Add_Vector(
     A=$Output.FN_HermiteCubicSpline.Position
     B=&Multiply_VectorByFloat(
      Vector=&RotateVectorByQuaternion(
       Vector To Rotate=&MakeVector(
        X=0
        Y=&Waveform(
         Global Input Position=&Multiply_Float(A=$Particles.ParticleDistance, B=0.005)
         Global Amplitude Scale=10
         "Waveform [1]"=::NewEnumerator0
         "[1] Frequency"=1
         "[1] Amplitude Min/Max"=<0 1>
         "[1] Bias"=-0.5
         "[2] Phase"=$Emitter.Age
         "[3] Phase"=$Emitter.Age
        )
        Z=&Waveform(
         Global Input Position=&Multiply_Float(A=$Particles.ParticleDistance, B=0.005)
         Global Amplitude Scale=10
         "Waveform [1]"=::NewEnumerator1
         "[1] Frequency"=1
         "[1] Amplitude Min/Max"=<0 1>
         "[1] Bias"=-0.5
         "[2] Phase"=$Emitter.Age
         "[3] Phase"=$Emitter.Age
        )
       )
       Quaternion=$System.BeamOrientation
      )
      Float=$Particles.WobbleOverLength
     )
    )) {isSetParameters=true, module=SetVariables_#7}
  ]
  rendererClasses [&NiagaraRibbonRendererProperties]
  renderers [
   {rendererClass=&NiagaraRibbonRendererProperties}
   {rendererClass=&NiagaraRibbonRendererProperties, rendererIndex=1}
  ]
 }
 NE_Flare {
  emitterUpdateScript [
   &EmitterState()
   &SpawnRate(SpawnRate=10, Use Spawn Probability=false)
  ]
  particleSpawnScript [
   &InitializeParticle(
    Color=<rgba 1 1 1 4>
    Lifetime=0.25
    Mass Mode=::NewEnumerator1
    Mesh Scale=&VectorFromFloat(Value=0.5)
    Mesh Scale Mode=::NewEnumerator4
    Sprite Rotation Mode=::NewEnumerator2
    Sprite Size Mode=::NewEnumerator3
    Uniform Sprite Size=75
    UsePositionOffset=true
    Write Material Random=true
   )
   &CameraOffset(Camera Offset Amount=15)
   &SubUVAnimation(
    Direct Frame Index=2
    "Sprite Renderer "=DataInterface() {propertyValues {SpriteRenderer=&self:NE_Flare_5.NiagaraSpriteRendererProperties_0}}
    SubUV Animation Mode:ENiagara_SubUVLookupModeV2=::NewEnumerator5
   )
  ]
  particleUpdateScript [
   &ParticleState()
   &DynamicMaterialParameters(
    Index 0 Param 1=2
    Index 0 Param 4=1
    Index 1 Param 1=2
    Index 1 Param 2=0.4
    Index 1 Param 3=0.6
    Index 1 Param 4=1
    Index 2 Param 1=&FloatFromCurve(FloatCurve=DataInterface() {propertyValues {Curve {keys [<at 0.46 0>,<at 1 1>]}}})
    Index 2 Param 2=45
    Module.Index0Float0Write=true
    Module.Index0Float1Write=true
    Module.Index0Float2Write=true
    Module.Index0Float3Write=true
    Module.Index1Float0Write=true
    Module.Index1Float1Write=true
    Module.Index1Float2Write=true
    Module.Index1Float3Write=true
    Module.Index2Float0Write=true
    Module.Index2Float1Write=true
    Module.Index2Float2Write=true
    Module.Index2Float3Write=true
    Write Parameter Index 1=::NewEnumerator1
    Write Parameter Index 2=::NewEnumerator1
   )
   &ScaleColor(
    Linear Color Curve=DataInterface() {propertyValues {
      AlphaCurve {keys [<at 0.4156 1>,<at 1 0>]}
      BlueCurve {keys [<at 0.1184 0>,<at 0.2689 0.0011>]}
      GreenCurve {keys [<at 0.1184 1>,<at 0.2689 0.8>]}
      RedCurve {keys [<at 0.1184 0.4851>,<at 0.2689 0.5679>]}
     }}
    Scale Mode=::NewEnumerator2
   )
   /SetVariables(Particles.Position=&AddVectorToPosition(
     Position=&TransformPosition(Position=Linked() {linkedVariable {name=User.BeamEnd_0, type=NiagaraPosition}}, Source Space=::World, Destination Space=::Local)
     Vector=<0 0 0>
    )) {isSetParameters=true, module=SetVariables_#4}
  ]
  rendererClasses [&NiagaraSpriteRendererProperties]
  renderers [{rendererClass=&NiagaraSpriteRendererProperties}]
 }
 NE_ImpactFlipBook {
  emitterUpdateScript [
   &EmitterState()
   &SpawnRate(SpawnRate=12, Use Spawn Probability=false)
  ]
  particleSpawnScript [
   &InitializeParticle(
    Lifetime=0.25
    Mass Mode=::NewEnumerator1
    Mesh Scale=&VectorFromFloat(Value=0.5)
    Mesh Scale Mode=::NewEnumerator4
    Sprite Rotation Mode=::NewEnumerator2
    Sprite Size Mode=::NewEnumerator3
    Sprite UV Mode=::NewEnumerator3
    Uniform Sprite Size=60
    UsePositionOffset=true
    Write Material Random=true
   )
   &CameraOffset(Camera Offset Amount=15)
  ]
  particleUpdateScript [
   &ParticleState()
   &ScaleSpriteSize(Uniform Curve Sprite Scale=DataInterface() {propertyValues {Curve {defaultValue=0, keys [{arriveTangent=0.6475, time=0, value=0.4}, {interpMode=RCIM_Cubic, leaveTangent=0.2037, time=1, value=1}]}}})
   &SubUVAnimation(
    Direct Frame Index:NiagaraFloat=&FloatFromCurve(FloatCurve=DataInterface() {propertyValues {Curve {keys [<at 0 3>,<at 1 5>]}}})
    "Sprite Renderer "=DataInterface() {propertyValues {SpriteRenderer=&self:NE_ImpactFlipBook_4.NiagaraSpriteRendererProperties_3}}
    SubUV Animation Mode:ENiagara_SubUVLookupModeV2=::NewEnumerator5
   )
   &ScaleColor(
    Linear Color Curve=DataInterface() {propertyValues {
      AlphaCurve {keys [<at 0.5886 1>,<at 0.9866 0>]}
      BlueCurve {keys [<at 0.1312 0.2437>,<at 0.5356 0.0017>]}
      GreenCurve {keys [<at 0.1312 2>,<at 0.5356 0.625>]}
      RedCurve {keys [<at 0.1312 0.6357>,<at 0.5356 0.1736>]}
     }}
    Scale Mode=::NewEnumerator2
   )
   &DynamicMaterialParameters(
    Index 0 Param 1=&FloatFromCurve(FloatCurve=DataInterface() {propertyValues {Curve {keys [
         {arriveTangent=-0.0044, interpMode=RCIM_Cubic, leaveTangent=-0.0044, tangentMode=RCTM_User, time=0, value=0}
         {arriveTangent=1.6055, tangentMode=RCTM_User, time=1, value=0.5}
        ]}}})
    Module.Index0Float0Write=true
    Module.Index0Float1Write=true
    Module.Index0Float2Write=true
    Module.Index0Float3Write=true
   )
   /SetVariables(Particles.Position=&AddVectorToPosition(
     Position=&TransformPosition(Position=Linked() {linkedVariable {name=User.BeamEnd_0, type=NiagaraPosition}}, Source Space=::World, Destination Space=::Local)
     Vector=<0 0 0>
    )) {isSetParameters=true, module=SetVariables_#4}
  ]
  rendererClasses [&NiagaraSpriteRendererProperties]
  renderers [{rendererClass=&NiagaraSpriteRendererProperties}]
 }
 RingsMeshes {
  emitterUpdateScript [
   &EmitterState(
    Inactive Response=::NewEnumerator1
    Life Cycle Mode=::NewEnumerator1
    Loop Duration=&Lerp_Float(
     A=0.25
     B=0.1
     Alpha=&RemapRange(Input Value To Remap=$System.BeamLength, Input Min=0, Input Max=7679, Output Min=0, Output Max=1)
    )
    UseLoopDelay=false
   )
   &SpawnBurst_Instantaneous(Spawn Count=3, Use Spawn Probability=false, UseLoopCountLimit=false)
   &SpawnBurst_Instantaneous(
    Spawn Group=1
    Spawn Time=&Lerp_Float(
     A=0.025
     B=0.0125
     Alpha=&RemapRange(Input Value To Remap=$System.BeamLength, Input Min=0, Input Max=7679, Output Min=0, Output Max=1)
    )
    Use Spawn Probability=false
    UseLoopCountLimit=false
   ) {module=SpawnBurst_Instantaneous001}
   &SpawnBurst_Instantaneous(
    Spawn Time=&Lerp_Float(
     A=0.05
     B=0.025
     Alpha=&RemapRange(Input Value To Remap=$System.BeamLength, Input Min=0, Input Max=7679, Output Min=0, Output Max=1)
    )
    Use Spawn Probability=false
    UseLoopCountLimit=false
   ) {module=SpawnBurst_Instantaneous002}
  ]
  particleSpawnScript [
   &InitializeParticle(
    Mesh Scale Mode=::NewEnumerator3
    Mesh Uniform Scale=&MaskFloat(Mask Source=0, Mask Parameter=?, Mask=0, Masked Float=0.2, Passthrough Float=0.3)
    Sprite Size Mode=::NewEnumerator3
    Uniform Sprite Size=&MaskFloat(Mask Source=0, Mask Parameter=?, Mask=0, Masked Float=50, Passthrough Float=100)
    UsePositionOffset=false
    Write Material Random=true
   )
   &InitialMeshOrientation(Mesh Axis To Orient=<0 0 1>, Mesh Orientation Mode=::NewEnumerator2, Orientation Vector=<0 1 0>, Use Rotation Vector=false)
  ]
  particleUpdateScript [
   &ParticleState()
   &FN_HermiteCubicSpline(
    Position Blend:NiagaraFloat=$Particles.NormalizedAge
    Position End:Vector3f=&TransformPosition(Position=Linked() {linkedVariable {name=User.BeamEnd_0, type=NiagaraPosition}}, Source Space=::World, Destination Space=::Local)
    Position Start:Vector3f=$System.BeamOriginOffset
    Write Particles Position=true
   )
   /SetVariables(Particles.WobbleOverLength:NiagaraFloat=&FloatFromCurve(FloatCurve=DataInterface() {propertyValues {Curve {keys [
         {arriveTangent=5.2404, interpMode=RCIM_Cubic, leaveTangent=5.2404, tangentMode=RCTM_User, time=0, value=0}
         <at 0.5 1 RCIM_Cubic>
         {arriveTangent=-4.4363, interpMode=RCIM_Cubic, leaveTangent=-4.4363, tangentMode=RCTM_User, time=1, value=0}
        ]}}})) {isSetParameters=true, module=SetVariables_#5}
   /SetVariables(Particles.ParticleDistance:NiagaraFloat=&Multiply_Float(
     A=&Abs_Float(Float:NiagaraFloat=&VectorLength(Vector=Linked() {linkedVariable {name=Particles.Position, type=NiagaraPosition}}))
     B=0.1
    )) {isSetParameters=true, module=SetVariables_#6}
   /SetVariables(Particles.Position=&Add_Vector(
     A=$Output.FN_HermiteCubicSpline.Position
     B=&Multiply_VectorByFloat(
      Vector=&RotateVectorByQuaternion(
       Vector To Rotate=&MakeVector(
        X=0
        Y=&Waveform(
         Global Input Position=&Multiply_Float(A=$Particles.ParticleDistance, B=0.005)
         Global Amplitude Scale=10
         "Waveform [1]"=::NewEnumerator0
         "[1] Frequency"=1
         "[1] Amplitude Min/Max"=<0 1>
         "[1] Bias"=-0.5
         "[2] Phase"=$Emitter.Age
         "[3] Phase"=$Emitter.Age
        )
        Z=&Waveform(
         Global Input Position=&Multiply_Float(A=$Particles.ParticleDistance, B=0.005)
         Global Amplitude Scale=10
         "Waveform [1]"=::NewEnumerator1
         "[1] Frequency"=1
         "[1] Amplitude Min/Max"=<0 1>
         "[1] Bias"=-0.5
         "[2] Phase"=$Emitter.Age
         "[3] Phase"=$Emitter.Age
        )
       )
       Quaternion=$System.BeamOrientation
      )
      Float=$Particles.WobbleOverLength
     )
    )) {isSetParameters=true, module=SetVariables_#7}
   &ScaleMeshSize(Scale Factor=&Multiply_Vector(
     A=<1.25 1.25 1.75>
     B=&MakeVector(
      X=&FloatFromCurve(FloatCurve=DataInterface() {propertyValues {Curve {
          defaultValue=1
          keys [
           {arriveTangent=0.0255, interpMode=RCIM_Cubic, time=0.05, value=0}
           {arriveTangent=0.1921, interpMode=RCIM_Cubic, leaveTangent=0.1921, tangentMode=RCTM_User, time=0.1, value=1}
           <at 0.2 0.8 RCIM_Cubic>
           <at 0.3 1 RCIM_Cubic>
           <at 0.4 0.8 RCIM_Cubic>
           <at 0.5 1 RCIM_Cubic>
           <at 0.6 0.8 RCIM_Cubic>
           <at 0.7 1 RCIM_Cubic>
           <at 0.8 0.8 RCIM_Cubic>
           {arriveTangent=-0.0463, interpMode=RCIM_Cubic, leaveTangent=-0.0463, tangentMode=RCTM_User, time=0.9, value=1}
           {interpMode=RCIM_Cubic, leaveTangent=-29.9336, time=1, value=0}
          ]
         }}})
      Y=&FloatFromCurve(FloatCurve=DataInterface() {propertyValues {Curve {
          defaultValue=1
          keys [
           {arriveTangent=29.0201, interpMode=RCIM_Cubic, time=0.05, value=0}
           {arriveTangent=0.1921, interpMode=RCIM_Cubic, leaveTangent=0.1921, tangentMode=RCTM_User, time=0.1, value=1}
           <at 0.2 0.8 RCIM_Cubic>
           <at 0.3 1 RCIM_Cubic>
           <at 0.4 0.8 RCIM_Cubic>
           <at 0.5 1 RCIM_Cubic>
           <at 0.6 0.8 RCIM_Cubic>
           <at 0.7 1 RCIM_Cubic>
           <at 0.8 0.8 RCIM_Cubic>
           {arriveTangent=-0.0463, interpMode=RCIM_Cubic, leaveTangent=-0.0463, tangentMode=RCTM_User, time=0.9, value=1}
           {interpMode=RCIM_Cubic, leaveTangent=-29.9336, time=1, value=0}
          ]
         }}})
      Z=1
     )
    ))
   &ScaleSpriteSize(Uniform Curve Sprite Scale=DataInterface() {propertyValues {Curve {
       defaultValue=0
       keys [
        {arriveTangent=6.3581, interpMode=RCIM_Cubic, time=0, value=0}
        {arriveTangent=30.877, tangentMode=RCTM_User, time=0.1, value=1}
        {arriveTangent=-23.2322, interpMode=RCIM_Cubic, leaveTangent=-23.2322, tangentMode=RCTM_User, time=0.9, value=1}
        <at 1 0 RCIM_Cubic>
       ]
      }}})
   /SetVariables(Particles.Soft Glow Color:LinearColor=&LerpLinearColors(StartColor=<rgba 0.0538 0.25 0 0.1>, EndColor=<rgba 0.5 0.1422 0 0.1>, LerpFactor=$User.ChargePercent)) {isSetParameters=true}
   &ScaleColor(
    Scale Mode=::NewEnumerator1
    Scale RGBA=&MakeVector4_FromLinearColor(LinearColor=&LerpLinearColors(StartColor=<rgba 0.4301 2 0 1>, EndColor=<rgba 2 0.5689 0 1>, LerpFactor=$User.ChargePercent))
    ScaleRGBA=true
   )
  ]
  rendererClasses [&NiagaraMeshRendererProperties, &NiagaraSpriteRendererProperties]
  renderers [
   {rendererClass=&NiagaraMeshRendererProperties, rendererIndex=0}
   {rendererClass=&NiagaraSpriteRendererProperties, rendererIndex=1}
  ]
 }
}
summary {
 systemName=PH_NS_WildEstate_CoveSpice_Beam_Burst
 emitters [
  {emitterName=MainRibbon, rendererClasses [&NiagaraRibbonRendererProperties]}
  {emitterName=RingsMeshes, rendererClasses [&NiagaraMeshRendererProperties, &NiagaraSpriteRendererProperties]}
  {emitterName=Flashes, rendererClasses [&NiagaraMeshRendererProperties]}
  {emitterName=NE_ImpactFlipBook, rendererClasses [&NiagaraSpriteRendererProperties]}
  {emitterName=NE_Flare, rendererClasses [&NiagaraSpriteRendererProperties]}
 ]
 userVariables [
  User.BeamEnd_0 {type=NiagaraPosition, description="", defaultValue {}}
  User.BeamStart_0_0 {type=Vector3f, description="", defaultValue {struct=Vector3f, value=<0 0 0>}}
  User.ChargePercent {type=NiagaraFloat, defaultValue=0, description=""}
 ]
}
